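# Sound cue for the LimeRock UFO sound effect of flying towards a tree.
# A single wave player feeds the UFO wave directly to the output at reduced volume, with no randomization or mixing.
# It uses the shared UFO action attenuation and the Fort_Events sound class.
# Per-owner concurrency limits are filled in, but the override switch is left off and the concurrency set holds only an empty entry.
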SoundCue /LimeRock/Assets/Audio/SFX/UFO/SC_LimeRock_100_UFO_FlyTowardsTree
root=WavePlayer_0
nodes [{id=WavePlayer_0, properties {soundWaveAssetPtr=/LimeRock/Assets/Audio/SFX/UFO/wav/IP_LimeRock_100_UFO_FlyTowardTheTree}}]
properties {
 attenuationSettings=/LimeRock/Assets/Audio/Settings/ATT_LimeRock_UFO_Action
 maxDistance=30000
 soundClassObject=&Fort_Events
 volumeMultiplier=0.6
 concurrencyOverrides {bLimitToOwner=true, maxCount=2, retriggerTime=0.25}
 concurrencySet [null]
}
