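# Looping flight sound cue for the LimeRock UFO, played at reduced volume in the Fort_Events sound class.
# A looping wave player feeds a Doppler node with raised intensity, shifting pitch with relative motion between source and listener.
# A dedicated UFO loop attenuation asset handles spatial falloff.
# The cue also carries owner-limited concurrency override settings, while its concurrency set holds only an empty entry.
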SoundCue /LimeRock/Assets/Audio/SFX/UFO/SC_LimeRock_100_UFO_Flying_LP
root=Doppler_0
nodes [
 {id=Doppler_0, children [WavePlayer_1], properties {dopplerIntensity=1.2}}
 {id=WavePlayer_1, properties {bLooping=true, soundWaveAssetPtr=/LimeRock/Assets/Audio/SFX/UFO/wav/IP_LimeRock_100_UFO_Flying_LP}}
]
properties {
 attenuationSettings=/LimeRock/Assets/Audio/Settings/ATT_LimeRock_UFO_LP
 maxDistance=23000
 soundClassObject=&Fort_Events
 volumeMultiplier=0.4
 concurrencyOverrides {bLimitToOwner=true, maxCount=2, retriggerTime=0.25}
 concurrencySet [null]
}
