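# Sound cue for the UFO beam-lift effect in the LimeRock pre-event, classed as an event sound.
# A mixer layers the main beam-lift wave over a quieter "Swt" layer, each through its own modulator for random pitch variation.
# It uses large point-of-interest attenuation and a light send to distant reverb.
# A concurrency override caps simultaneous voices and applies volume scaling as instances stack.
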SoundCue /LimeRock/Assets/Audio/SFX/UFO/SC_LimeRock_PreEvent_UFO_BeamLift
root=Mixer_0
nodes [
 {id=Mixer_0, children [Modulator_0,Modulator_1], properties {inputVolume [1.2,0.6]}}
 {id=Modulator_0, children [WavePlayer_1], properties {pitchMax=1, pitchMin=0.85, volumeMax=1, volumeMin=1}}
 {id=Modulator_1, children [WavePlayer_0], properties {pitchMax=1.1, pitchMin=0.9, volumeMax=1, volumeMin=1}}
 {id=WavePlayer_0, properties {soundWaveAssetPtr=/LimeRock/Assets/Audio/SFX/UFO/wav/IP_LimeRock_PreEvent_UFO_BeamLift_Swt}}
 {id=WavePlayer_1, properties {soundWaveAssetPtr=/LimeRock/Assets/Audio/SFX/UFO/wav/IP_LimeRock_PreEvent_UFO_BeamLift}}
]
properties {
 attenuationSettings=/LimeRock/Assets/Audio/Settings/ATT_LimeRock_POI_Large
 bOverrideConcurrency=true
 maxDistance=40000
 soundClassObject=&Fort_Events
 volumeMultiplier=0.5
 concurrencyOverrides {maxCount=3, voiceStealReleaseTime=0.5, volumeScale=0.75, volumeScaleAttackTime=0.25}
 soundSubmixSends [{minSendLevel=0, sendLevel=0.1, soundSubmix=&Fortnite_Reverb_Distant}]
}
