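# Sound cue for the third-person UFO beam lift effect in the LimeRock pre-event sequence.
# A single wave plays through a modulator that randomizes volume and pitch slightly below unity.
# It uses the pre-event UFO attenuation and the Fort_Events sound class.
# Overridden concurrency caps simultaneous voices, stealing the farthest and scaling down volume as copies stack.
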SoundCue /LimeRock/Assets/Audio/SFX/UFO/SC_LimeRock_PreEvent_UFO_BeamLift_3P_V2
root=Modulator_0
nodes [
 {id=Modulator_0, children [WavePlayer_0], properties {pitchMax=0.9, pitchMin=0.8, volumeMax=0.9, volumeMin=0.8}}
 {id=WavePlayer_0, properties {soundWaveAssetPtr=/LimeRock/Assets/Audio/SFX/UFO/wav/IP_LimeRock_PreEvent_UFO_BeamLift_3P}}
]
properties {
 attenuationSettings=/LimeRock/Assets/Audio/Settings/ATT_LimeRock_PreEvent_UFO
 bOverrideConcurrency=true
 maxDistance=12033.3789
 soundClassObject=&Fort_Events
 volumeMultiplier=0.8
 concurrencyOverrides {maxCount=3, resolutionRule=StopFarthestThenPreventNew, voiceStealReleaseTime=0.5, volumeScale=0.75, volumeScaleAttackTime=0.25}
}
