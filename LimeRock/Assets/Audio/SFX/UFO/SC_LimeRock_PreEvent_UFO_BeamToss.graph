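# Sound cue for the UFO beam toss sound effect in the LimeRock pre-event, routed to the Fort_Events sound class.
# A single wave player feeds the output at reduced volume, using the large point-of-interest attenuation with a long max distance.
# A concurrency override caps simultaneous instances, with a voice-steal release and volume scaling for overlapping plays.
# A low-level send feeds the distant reverb submix.
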SoundCue /LimeRock/Assets/Audio/SFX/UFO/SC_LimeRock_PreEvent_UFO_BeamToss
root=WavePlayer_0
nodes [{id=WavePlayer_0, properties {soundWaveAssetPtr=/LimeRock/Assets/Audio/SFX/UFO/wav/IP_LimeRock_PreEvent_UFO_BeamToss}}]
properties {
 attenuationSettings=/LimeRock/Assets/Audio/Settings/ATT_LimeRock_POI_Large
 bOverrideConcurrency=true
 maxDistance=40000
 soundClassObject=&Fort_Events
 volumeMultiplier=0.5
 concurrencyOverrides {maxCount=3, voiceStealReleaseTime=0.5, volumeScale=0.75, volumeScaleAttackTime=0.25}
 soundSubmixSends [{minSendLevel=0, sendLevel=0.1, soundSubmix=&Fortnite_Reverb_Distant}]
}
