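# SoundCue for the UFO dive-bomb sweetener effect in the LimeRock pre-event.
# A Random node picks one of three equally weighted wave variations, and a Modulator slightly randomizes its pitch at constant volume.
# It plays in the Fort_Events sound class with shared UFO action attenuation and a slightly reduced volume.
# Concurrency overrides cap simultaneous instances per owner and set a short retrigger time.
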SoundCue /LimeRock/Assets/Audio/SFX/UFO/SC_LimeRock_PreEvent_UFO_DiveBomb_Swt
root=Modulator_0
nodes [
 {id=Modulator_0, children [Random_0], properties {pitchMax=1.1, pitchMin=0.9, volumeMax=1, volumeMin=1}}
 {id=Random_0, children [WavePlayer_1,WavePlayer_2,WavePlayer_3], properties {weights [1,1,1]}}
 {id=WavePlayer_1, properties {soundWaveAssetPtr=/LimeRock/Assets/Audio/SFX/UFO/wav/IP_LimeRock_PreEvent_UFO_DiveBomb_Swt_01}}
 {id=WavePlayer_2, properties {soundWaveAssetPtr=/LimeRock/Assets/Audio/SFX/UFO/wav/IP_LimeRock_PreEvent_UFO_DiveBomb_Swt_02}}
 {id=WavePlayer_3, properties {soundWaveAssetPtr=/LimeRock/Assets/Audio/SFX/UFO/wav/IP_LimeRock_PreEvent_UFO_DiveBomb_Swt_03}}
]
properties {
 attenuationSettings=/LimeRock/Assets/Audio/Settings/ATT_LimeRock_PreEvent_UFO_Action
 maxDistance=30000
 soundClassObject=&Fort_Events
 volumeMultiplier=0.8
 concurrencyOverrides {bLimitToOwner=true, maxCount=2, retriggerTime=0.25}
 concurrencySet [null]
}
