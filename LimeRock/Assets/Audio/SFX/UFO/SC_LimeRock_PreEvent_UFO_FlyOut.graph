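# Sound cue for the UFO fly-out sound effect in the LimeRock pre-event.
# A single wave player plays one fly-out wave, with boosted volume and a dedicated pre-event UFO attenuation asset.
# It routes to the Fort_Events sound class and carries concurrency overrides that cap instances per owner with a short retrigger time.
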SoundCue /LimeRock/Assets/Audio/SFX/UFO/SC_LimeRock_PreEvent_UFO_FlyOut
root=WavePlayer_0
nodes [{id=WavePlayer_0, properties {soundWaveAssetPtr=/LimeRock/Assets/Audio/SFX/UFO/wav/IP_LimeRock_PreEvent_UFO_FlyOut_01}}]
properties {
 attenuationSettings=/LimeRock/Assets/Audio/Settings/ATT_LimeRock_PreEvent_UFO
 maxDistance=12033.3789
 soundClassObject=&Fort_Events
 volumeMultiplier=1.5
 concurrencyOverrides {bLimitToOwner=true, maxCount=2, retriggerTime=0.25}
 concurrencySet [null]
}
