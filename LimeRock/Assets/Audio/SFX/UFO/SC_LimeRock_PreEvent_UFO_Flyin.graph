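# Sound cue for the UFO fly-in sound effect in the LimeRock pre-event.
# A single wave player feeds the output directly, with no randomization or modulation.
# It uses the pre-event UFO action attenuation and the Fort_Events sound class, with boosted volume.
# Concurrency override values limit copies per owner and set a retrigger time, while the concurrency set holds only an empty entry.
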SoundCue /LimeRock/Assets/Audio/SFX/UFO/SC_LimeRock_PreEvent_UFO_Flyin
root=WavePlayer_1
nodes [{id=WavePlayer_1, properties {soundWaveAssetPtr=/LimeRock/Assets/Audio/SFX/UFO/wav/IP_LimeRock_PreEvent_UFO_FlyIn_01}}]
properties {
 attenuationSettings=/LimeRock/Assets/Audio/Settings/ATT_LimeRock_PreEvent_UFO_Action
 maxDistance=30000
 soundClassObject=&Fort_Events
 volumeMultiplier=1.5
 concurrencyOverrides {bLimitToOwner=true, maxCount=2, retriggerTime=0.25}
 concurrencySet [null]
}
